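# Post-process style cosmetic AnimBlueprint for the large male SnailAisle body that follows the parent mesh and adjusts its shoulder pads.
# Copies the attached parent's pose through the SnailAisle Control Rig, fed the skydiving state as inAir, then caches it.
# Additive bone-space transforms offset both shoulder-pad bones, blending in while crouch-moving or while jumping or falling outside skydiving.
# The event graph derives those two booleans each update.
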AnimBlueprint /BRCosmetics/Characters/Player/Male/Large/Bodies/M_LRG_SnailAisle/Meshes/M_LRG_SnailAisle_AnimBP
target_skeleton=&/M_LRG_SnailAisle_Skeleton
graphs [
 AnimGraph {nodes [
   n1 = Get CrouchMoving() -> (CrouchMoving:bool)
   n2 = "Transform (Modify) Bone - Bone: dyn_shoulderpad_r"!(
    ComponentPose=Local To Component(LocalPose=Use cached pose 'CRCache_Input'())
    bAlphaBoolEnabled:bool=@n1
    Translation="1.895,-0.000,-0.549"
    Rotation="-16.171,0.000,-0.000"
    Scale="0.95,1,1"
   ) -> (Pose) {
    alphaInputType=Bool
    rotationMode=BMM_Additive
    rotationSpace=BCS_BoneSpace
    scaleMode=BMM_Additive
    scaleSpace=BCS_BoneSpace
    translationMode=BMM_Additive
    translationSpace=BCS_BoneSpace
    alphaBoolBlend {blendInTime=0.4, blendOption=Linear, blendOutTime=0.2}
    boneToModify {boneName=dyn_shoulderpad_r}
    rotation=<rot -16.171 0.0001 -0.0001>
    scale {x=0.95}
    translation {x=1.895, z=-0.549}
   }
   n3 = "Transform (Modify) Bone - Bone: dyn_shoulderpad_l"!(ComponentPose=@n2, bAlphaBoolEnabled:bool=@n1, Translation="-1.895,0,-0.549", Rotation="-16.171,0.000,-0.000", Scale="0.95,1,1") -> (Pose) {
    alphaInputType=Bool
    rotationMode=BMM_Additive
    rotationSpace=BCS_BoneSpace
    scaleMode=BMM_Additive
    scaleSpace=BCS_BoneSpace
    translationMode=BMM_Additive
    translationSpace=BCS_BoneSpace
    alphaBoolBlend {blendInTime=0.4, blendOption=Linear, blendOutTime=0.2}
    boneToModify {boneName=dyn_shoulderpad_l}
    rotation=<rot -16.171 0.0001 -0.0001>
    scale {x=0.95}
    translation {x=-1.895, z=-0.549}
   }
   n4 = Get jumpFall() -> (jumpFall:bool)
   n5 = "Transform (Modify) Bone - Bone: dyn_shoulderpad_r"(ComponentPose=@n3, bAlphaBoolEnabled=@n4, Translation="4.318,-0.388,0.347", Rotation="-7.413,-1.250,0.192", Scale="0.9,1,1") {
    alphaInputType=Bool
    rotationMode=BMM_Additive
    rotationSpace=BCS_BoneSpace
    scaleMode=BMM_Additive
    scaleSpace=BCS_BoneSpace
    translationMode=BMM_Additive
    translationSpace=BCS_BoneSpace
    alphaBoolBlend {blendInTime=0.2, blendOption=Linear, blendOutTime=0.2}
    boneToModify {boneName=dyn_shoulderpad_r}
    rotation=<rot -7.413 -1.25 0.192>
    scale {x=0.9}
    translation=<4.318 -0.388 0.347>
   }
   n6 = "Transform (Modify) Bone - Bone: dyn_shoulderpad_l"(ComponentPose=@n5, bAlphaBoolEnabled=@n4, Translation="-3.3,-0.25,0.1", Rotation="-7,-0.9,-0.18", Scale="0.9,1,1") {
    alphaInputType=Bool
    rotationMode=BMM_Additive
    rotationSpace=BCS_BoneSpace
    scaleMode=BMM_Additive
    scaleSpace=BCS_BoneSpace
    translationMode=BMM_Additive
    translationSpace=BCS_BoneSpace
    alphaBoolBlend {blendInTime=0.2, blendOption=Linear, blendOutTime=0.2}
    boneToModify {boneName=dyn_shoulderpad_l}
    rotation=<rot -7 -0.9 -0.18>
    scale {x=0.9}
    translation=<-3.3 -0.25 0.1>
   }
   Output Pose(Result=Component To Local(ComponentPose=@n6)) {layerGroup=None}
   n7 = Control Rig!(Source=Copy Pose From Mesh() {bUseAttachedParent=true}, Alpha, inAir:bool=Get bIsSkyDiving()) -> (Pose) {
    alphaBoolBlend {blendOption=Linear}
    controlRigAssetReference {blueprintRigClass=&/M_LRG_SnailAisle_CtrlRig.M_LRG_SnailAisle_CtrlRig_C}
   }
   Save cached pose 'CRCache_Input'(Pose=@n7) {cacheName=CRCache_Input}
  ]}
 EventGraph {nodes [
   Event BlueprintUpdateAnimation()
   > Set CrouchMoving(CrouchMoving:bool=AND Boolean(A=Get bIsCrouching(), B=Get bIsMoving2D())) -> (Output_Get:bool)
   n1 = Get bIsSkyDiving()
   n2 = Get bIsJumping()
   n3 = Get bIsFalling()
   > Set jumpFall(jumpFall:bool=AND Boolean(A=OR Boolean(A=@n2, B=@n3), B=NOT Boolean(A=@n1))) -> (Output_Get:bool)
  ]}
]
variables [
 CrouchMoving {type=boolean}
 jumpFall {type=boolean}
]
