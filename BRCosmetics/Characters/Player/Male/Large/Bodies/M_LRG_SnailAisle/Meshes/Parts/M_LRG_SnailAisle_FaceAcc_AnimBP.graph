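# Post-copy anim blueprint for the SnailAisle large male face accessory, adding secondary motion to its hood.
# It copies the attached parent's pose, parent-constrains the hood-back base to weighted head, spine, clavicle and hood-tip bones, then additively offsets and scales dyn_hoodback_m.
# Chained AnimDynamics nodes with angular springs simulate the hood tip and sides.
# The event graph computes headLookUp from head-to-neck rotation, but the AnimGraph never reads it.
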AnimBlueprint /BRCosmetics/Characters/Player/Male/Large/Bodies/M_LRG_SnailAisle/Meshes/Parts/M_LRG_SnailAisle_FaceAcc_AnimBP
target_skeleton=&/M_LRG_SnailAisle_FaceAcc_Skeleton
graphs [
 AnimGraph {nodes [
   n1 = "Constraint - base_hoodback_m (PPPPPP)"!(
    ComponentPose=Local To Component(LocalPose=Copy Pose From Mesh() {bUseAttachedParent=true})
    Alpha=1
    ConstraintWeights_0=0.5115
    ConstraintWeights_1:float=0.3346
    ConstraintWeights_2:float=0.1326
    ConstraintWeights_3:float=0.011
    ConstraintWeights_4:float=0.0091
    ConstraintWeights_5:float=0.0012
   ) -> (Pose) {
    alphaBoolBlend {blendOption=Linear}
    boneToModify {boneName=base_hoodback_m}
    constraintSetup [
     {targetBone {boneName=head}}
     {targetBone {boneName=spine_05}}
     {targetBone {boneName=dyn_hoodTip_m}}
     {targetBone {boneName=clavicle_r}}
     {targetBone {boneName=clavicle_l}}
     {targetBone {boneName=spine_03}}
    ]
    constraintWeights [1,1,1,1,1,1]
   }
   n2 = "Transform (Modify) Bone - Bone: dyn_hoodback_m"(ComponentPose=@n1, Alpha=1, Translation="-2.173,0.004,6.628", Rotation="0,0,0", Scale="0.106,1.000,1.000") {
    rotationMode=BMM_Additive
    rotationSpace=BCS_BoneSpace
    scaleMode=BMM_Additive
    scaleSpace=BCS_BoneSpace
    translationMode=BMM_Additive
    translationSpace=BCS_BoneSpace
    alphaBoolBlend {blendOption=Linear}
    boneToModify {boneName=dyn_hoodback_m}
    scale {x=0.106}
    translation=<-2.173 0.004 6.628>
   }
   n3 = Anim Dynamics - Bone: dyn_hoodTip_m(ComponentPose=@n2, Alpha=1, ExternalForce="0,0,0") {
    angularDampingOverride=1
    angularSpringConstant=14000
    bAngularSpring=true
    bOverrideAngularDamping=true
    bUseGravityOverride=true
    gravityScale=0.35
    simulationSpace=RootRelative
    alphaBoolBlend {blendOption=Linear}
    boundBone {boneName=dyn_hoodTip_m}
    componentAppliedLinearAccClamp=<3000 3000 3000>
    componentLinearAccScale=<0.25 0.25 0.25>
    componentLinearVelScale=<0.25 0.25 0.25>
    gravityOverride {z=-980}
    physicsBodyDefinitions [{
      boundBone {boneName=dyn_hoodTip_m}
      boxExtents=<20 30 30>
      constraintSetup {angularLimitsMax=<0 10 15>, angularLimitsMin=<0 -20 -15>, angularTarget=<1 0 0>}
      localJointOffset=<1 0 0>
     }]
   }
   n4 = Anim Dynamics - Bone: dyn_hood_l(ComponentPose=@n3, Alpha=1, ExternalForce="0,0,0") {
    angularDampingOverride=1
    angularSpringConstant=20000
    bAngularSpring=true
    bOverrideAngularDamping=true
    bUseGravityOverride=true
    gravityScale=0.35
    simulationSpace=RootRelative
    alphaBoolBlend {blendOption=Linear}
    boundBone {boneName=dyn_hood_l}
    componentLinearAccScale=<1 1 1>
    componentLinearVelScale=<1 1 1>
    gravityOverride {z=-980}
    physicsBodyDefinitions [{
      boundBone {boneName=dyn_hood_l}
      boxExtents=<20 30 30>
      constraintSetup {angularLimitsMax=<0 15 15>, angularLimitsMin=<0 -15 -15>, angularTarget=<1 0 0>}
      localJointOffset=<4 0 0>
     }]
   }
   n5 = Anim Dynamics - Bone: dyn_hood_r(ComponentPose=@n4, Alpha=1, ExternalForce="0,0,0") {
    angularDampingOverride=1
    angularSpringConstant=20000
    bAngularSpring=true
    bOverrideAngularDamping=true
    bUseGravityOverride=true
    gravityScale=0.35
    simulationSpace=RootRelative
    alphaBoolBlend {blendOption=Linear}
    boundBone {boneName=dyn_hood_r}
    componentLinearAccScale=<1 1 1>
    componentLinearVelScale=<1 1 1>
    gravityOverride {z=-980}
    physicsBodyDefinitions [{
      boundBone {boneName=dyn_hood_r}
      boxExtents=<20 30 30>
      constraintSetup {angularLimitsMax=<0 15 15>, angularLimitsMin=<0 -15 -15>, angularTarget=<1 0 0>}
      localJointOffset=<4 0 0>
     }]
   }
   Output Pose(Result=Component To Local(ComponentPose=@n5)) {layerGroup=None}
  ]}
 EventGraph {nodes [
   Event BlueprintUpdateAnimation()
   > n1 = Set Skel Mesh(Skel Mesh=GetOwnerSkeletalMeshComponent())
   > Set headLookUp(headLookUp:double=GetAlphaFromReferenceTransformDelta(SkeletalMeshComponent=@n1, BoneName=head, BaseName=neck_01, InputRange=(X=0,Y=-20), TwistAxis="0,0,1")) -> (Output_Get:double)
  ]}
]
variables [
 headLookUp {type=number}
 Skel Mesh {type=SkeletalMeshComponent}
]
